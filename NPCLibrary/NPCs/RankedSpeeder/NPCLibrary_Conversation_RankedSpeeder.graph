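# Conversation database for the RankedSpeeder NPC in NPCLibrary, providing its opening greeting and top-level conversation branches.
# The entry point leads to one speech node with a default challenge line and contextual lines for first-ever conversations or specific player outfits, matched by character ID.
# Curve-table priority rows rank those contextual lines.
# The greeting branches through link nodes to the shared NPCLibrary quests and BasicServices entries and to RankedSpeeder's own Dialogue entry.
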ConversationDatabase /NPCLibrary/NPCs/RankedSpeeder/NPCLibrary_Conversation_RankedSpeeder
entry_points [{tag=Conversation.Entry.RankedSpeeder, destinations [#1]}]
nodes [
 ConversationEntryPointNode_7 {
  id=#1
  class=ConversationEntryPointNode
  outputs [#2]
  properties {entryTag {tagName=Conversation.Entry.RankedSpeeder}}
 }
 Speech_C_4 {
  id=#2
  class=Speech_C
  outputs [#3, #4, #5]
  properties {
   nodeName=""
   generalConfig {
    defaultMessage="If you're looking for a fight, let's be quick about it."
    contextualMessages [
     {
      message="Ready to ride, or are you walking?"
      contextRequirements [{requirement {class=FortControllerRequirement_IsFirstEverConversationWithNPC}}]
      priority {curve {rowName=Tandem.Speech.Priority.FirstEverInteraction}}
     }
     {
      message="Finally, a challenge. Looking for a race?"
      contextRequirements [{requirement {
         class=FortControllerRequirement_HasCID
         properties {allowedCIds [/BRCosmetics/Athena/Items/Cosmetics/Characters/CID_521_Athena_Commando_M_TacticalBiker]}
        }}]
      priority {curve {rowName=Tandem.Speech.Priority.OutfitReaction}}
     }
     {
      message=I've never lost a race. I won't start now.
      contextRequirements [{requirement {
         class=FortControllerRequirement_HasCID
         properties {allowedCIds [/BRCosmetics/Athena/Items/Cosmetics/Characters/CID_094_Athena_Commando_M_Rider]}
        }}]
      priority {curve {rowName=Tandem.Speech.Priority.OutfitReaction}}
     }
     {
      message="Think you're faster than me? Let's find out!"
      contextRequirements [{requirement {
         class=FortControllerRequirement_HasCID
         properties {allowedCIds [/BRCosmetics/Athena/Items/Cosmetics/Characters/CID_162_Athena_Commando_F_StreetRacer]}
        }}]
      priority {curve {rowName=Tandem.Speech.Priority.OutfitReaction}}
     }
     {
      message=Take a look. I'm what REAL speed looks like.
      contextRequirements [{requirement {
         class=FortControllerRequirement_HasCID
         properties {allowedCIds [/BRCosmetics/Athena/Items/Cosmetics/Characters/CID_524_Athena_Commando_F_FutureBikerWhite]}
        }}]
      priority {curve {rowName=Tandem.Speech.Priority.OutfitReaction}}
     }
    ]
   }
  }
 }
 ConversationLinkNode_2 {
  id=#5
  class=ConversationLinkNode
  properties {remoteEntryTag {tagName=Conversation.Entry.RankedSpeeder.Dialogue}}
 }
 ConversationLinkNode_0 {id=#4, class=ConversationLinkNode, properties {remoteEntryTag {tagName=Conversation.Entry.BasicServices}}}
 ConversationLinkNode_1 {id=#3, class=ConversationLinkNode, properties {remoteEntryTag {tagName=Conversation.Entry.NPCLibrary.Quests}}}
]
